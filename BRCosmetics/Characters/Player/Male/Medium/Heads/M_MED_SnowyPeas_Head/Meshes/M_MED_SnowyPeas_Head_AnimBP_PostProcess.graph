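# Post-process AnimBP for the SnowyPeas male head mesh that counter-rotates head and neck bones to damp their motion.
# Each update reads head, neck_02 and neck_01 deltas from ref pose relative to their parent bones, negates them and scales per axis by percentage variables.
# The AnimGraph adds these in bone space to head and neck_02; the neck_01 node is unconnected.
# The emote-check Branch has no incoming exec link.
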AnimBlueprint /BRCosmetics/Characters/Player/Male/Medium/Heads/M_MED_SnowyPeas_Head/Meshes/M_MED_SnowyPeas_Head_AnimBP_PostProcess
target_skeleton=/Game/Characters/Player/Common/Fortnite_Base_Head/Fortnite_Base_Head_Export_Skeleton
graphs [
 AnimGraph {nodes [
   "Transform (Modify) Bone - Bone: neck_01"(Alpha=1, Translation="0,0,0", Rotation=Get neck_1_rot() -> (neck_1_rot:Rotator), Scale="1,1,1") {rotationMode=BMM_Additive, rotationSpace=BCS_BoneSpace, alphaBoolBlend {blendOption=Linear}, boneToModify {boneName=neck_01}}
   n1 = "Transform (Modify) Bone - Bone: head"(
    ComponentPose=Local To Component(LocalPose=Input Pose())
    Alpha=1
    Translation="0,0,0"
    Rotation=Get head rot() -> (head rot:Rotator)
    Scale="1,1,1"
   ) {
    rotationMode=BMM_Additive
    rotationSpace=BCS_BoneSpace
    alphaBoolBlend {blendOption=Linear}
    boneToModify {boneName=head}
    rotation=<rot 50 50 50>
   }
   n2 = "Transform (Modify) Bone - Bone: neck_02"(ComponentPose=@n1, Alpha=1, Translation="0,0,0", Rotation=Get neck_2_rot() -> (neck_2_rot:Rotator), Scale="1,1,1") {rotationMode=BMM_Additive, rotationSpace=BCS_BoneSpace, alphaBoolBlend {blendOption=Linear}, boneToModify {boneName=neck_02}}
   Output Pose(Result=Component To Local(ComponentPose=@n2))
  ]}
 EventGraph {nodes [
   n1 = Event BlueprintUpdateAnimation()
   n2 = GetOwnerSkeletalMeshComponent()
   n3 = BreakRotator(InRot=BreakTransform(InTransform=GetDeltaTransformFromRefPose(self=@n2, BoneName=head, BaseName=neck_02)).Rotation)
   n4 = Get Z percentage() -> (Z percentage:float)
   n5 = Get X percentage() -> (X percentage:float)
   n6 = Get Y percentage() -> (Y percentage:float)
   Set head rot(
    execute=[@n1.then, @n17.then]
    head rot:Rotator=MakeRotator(
     Roll=float * float!(A=@n3, B=-1, C:double=@n5) -> (ReturnValue)
     Pitch=float * float(A=@n3.Pitch, B=-1, C=@n6)
     Yaw=float * float(A=@n3.Yaw, B=-1, C=@n4)
    )
   ) -> (Output_Get:Rotator)
   n7 = GetOwnerSkeletalMeshComponent()
   n8 = BreakRotator(InRot=BreakTransform(InTransform=GetDeltaTransformFromRefPose(self=@n7, BoneName=neck_02, BaseName=neck_01)).Rotation)
   n9 = Get Z percentage()
   n10 = Get X percentage()
   n11 = Get Y percentage()
   > Set neck_2_rot(neck_2_rot:Rotator=MakeRotator(Roll=float * float(A=@n8, B=-1, C=@n10), Pitch=float * float(A=@n8.Pitch, B=-1, C=@n11), Yaw=float * float(A=@n8.Yaw, B=-1, C=@n9))) -> (Output_Get:Rotator)
   n12 = GetOwnerSkeletalMeshComponent()
   n13 = BreakRotator(InRot=BreakTransform(InTransform=GetDeltaTransformFromRefPose(self=@n12, BoneName=neck_01, BaseName=spine_05)).Rotation)
   n14 = Get Y percentage()
   n15 = Get X percentage()
   n16 = Get Z percentage()
   > Set neck_1_rot(neck_1_rot:Rotator=MakeRotator(Roll=float * float(A=@n13, B=-1, C=@n15), Pitch=float * float(A=@n13.Pitch, B=-1, C=@n14), Yaw=float * float(A=@n13.Yaw, B=-1, C=@n16))) -> (Output_Get:Rotator)
   n17 = Branch(Condition=NOT Boolean(A=Get bIsPlayingEmote()))
   TryGetPawnOwner()
  ]}
]
variables [
 head rot {type=Rotator}
 neck_1_rot {type=Rotator}
 neck_2_rot {type=Rotator}
 X percentage {type=number, default=0.5}
 Y percentage {type=number, default=0.5}
 Z percentage {type=number, default=0.5}
]
